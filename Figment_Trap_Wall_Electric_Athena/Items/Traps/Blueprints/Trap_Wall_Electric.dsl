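; Electric wall trap Blueprint that handles trigger registration and audiovisual feedback over the trap's lifecycle.
; Placement adds the trigger component and dust; finished building, world ready and reload end play activation sound and start idle arcs.
; The ActivateTrap GameplayCue plays fire effects, boosts idle particles and fades a light by timeline, with a Low-quality branch cutting light and arcs.
; Running out of durability plays a delayed explosion sound and borrowed Dart Trap destruction effect.
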(event EventOnPlaced
  ;[1 1 1] Initial Setup
  (Utilities|FlowControl|DoOnce false
    (:Completed
      (Trap|AddTriggerComponent (Variables|Default|GetTriggerComponent))
      (Utilities|TrapPlacementDustFX :Attach%20to%20Component (Variables|Debug|GetStaticMeshComponent) :Socket "PlacedFX" :Rotation "0, 0, 0" :Burst%20Rate 10 :Size%20%28Volume%29 "100.000000,40.000000,15.000000" :Velocity%20Scale "0.500000,0.500000,0.500000" :Size%20Scale%20%28Particles%29 1))))

(event EventOnFinishedBuilding
  ;[1 1 1] Finished building
  (CallFunction|PlayActivationSoundandFX))

(event EventOnWorldReady
  ;[1 1 1] Finished building
  (CallFunction|PlayActivationSoundandFX))

(event GameplayCue|GameplayCue.Abilities.Activation.Traps.ActivateTrap
  (params (EventType :type EGameplayCueEvent%20Enum :default "OnActive") (Parameters :type Gameplay%20Cue%20Parameters%20Structure%20%28by%20ref%29))
  ;[1 1 1] On Fire
  (switch Utilities|FlowControl|Switch|SwitchonEGameplayCueEvent EventType
    (:Executed
      (Audio|PlaySoundatLocation (Variables|Audio|GetTrap_Fire_Sound_Cue) (Transformation|GetActorLocation)) ; Trap_Fire_Sound
      (Effects|SpawnEmitterAttached "/Figment_Trap_Wall_Electric_Athena/Weapons/FORT_TRAPS/Wall_Electric_Trap/P_ElectricWallTrap_Fire_01.P_ElectricWallTrap_Fire_01" (Variables|Debug|GetStaticMeshComponent) "FireFX_L")
      (Effects|SpawnEmitterAttached "/Figment_Trap_Wall_Electric_Athena/Weapons/FORT_TRAPS/Wall_Electric_Trap/P_ElectricWallTrap_Fire_01.P_ElectricWallTrap_Fire_01" (Variables|Debug|GetStaticMeshComponent) "FireFX_R")
      (Development|SetHiddeninGame (Variables|Default|GetLight_onFire) false false)
      (Effects|Components|ParticleSystem|SetFloatParameter (Variables|Trap_Wall_Electric|GetIdleFX_Arcs_L) "SpawnRateScale" 100)
      (Effects|Components|ParticleSystem|SetFloatParameter (Variables|Default|GetIdleFX_Coil_L) "SpawnRateScale" 10)
      (Utilities|EffectsQualityCheck 0
        (:Cinematic
          _)
        (:Epic
          _)
        (:High
          _)
        (:Medium
          _)
        (:Low
          (Development|SetHiddeninGame (Variables|Default|GetLight_onFire) true false)
          (Effects|Components|ParticleSystem|SetFloatParameter (Variables|Trap_Wall_Electric|GetIdleFX_Arcs_L) "SpawnRateScale" 1)
          (Components|Activation|Deactivate (Variables|Trap_Wall_Electric|GetIdleFX_Arcs_L))))
      (bind (_direction _brightness) (Timeline|FadeLight PlayFromStart)
        (:Update
          (Rendering|Components|Light|SetIntensity (Variables|Default|GetLight_onFire) _brightness))
        (:Finished)))))

(event EventOnReloadEnd
  ;[1 1 1] Reload End
  (Audio|PlaySoundatLocation (Variables|Audio|GetActivate_Sound_Cue) (Transformation|GetActorLocation)) ; Active Trap Sound
  (Components|Activation|Activate (Variables|Trap_Wall_Electric|GetIdleFX_Arcs_L) false))

(event EventOnOutOfDurability
  ;[1 1 1] Durability Change
  (Utilities|FlowControl|Delay 1)
  (Audio|PlaySoundatLocation (Variables|Audio|GetExplode_Sound_Cue) (Transformation|GetActorLocation)) ; Death_Sound
  (Effects|SpawnEmitteratLocation "/Figment_Trap_Wall_Electric_Athena/Weapons/FORT_TRAPS/Dart_Trap/P_Dart_Trap_Destruction.P_Dart_Trap_Destruction" (Transformation|GetWorldLocation (Building|GetBuildingMeshComponent))))

(fn PlayActivationSoundAndFX
  (Audio|PlaySoundatLocation (Variables|Audio|GetActivate_Sound_Cue) (Transformation|GetActorLocation)) ; Trap Active Sound
  (Components|Activation|Activate (Variables|Trap_Wall_Electric|GetIdleFX_Arcs_L) false))

(fn ConstructionScript
  (Networking|CanExecuteCosmeticEvents
    (:True
      (Development|SetHiddeninGame (Variables|Default|GetLight_onFire) true false))))
